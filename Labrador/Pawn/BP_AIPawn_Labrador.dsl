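; AI pawn for the Labrador llama, which can be toppled and spawns tiered loot on death.
; Accumulated damage past a random threshold launches and topples it, then temporarily swaps in a buff threshold.
; Knockback, fast vehicle hits and landings after launch or teleport can also topple it unless ridden; entering water clears this.
; State setters update the AI blackboard, while the topple OnRep toggles capsule collision and triggers animation input.
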(event EventBeginPlay
  (|CallParentFunction)
  (Variables|Default|SetDamageTreshold (Math|Random|RandomFloatinRange (ScalableFloat|GetValueatLevel (Variables|Default|GetInitDamageThresholdMin) 1.0) (ScalableFloat|GetValueatLevel (Variables|Default|GetInitDamageTresholdMax) 1.0)))
  (Ability|Tasks|WaitGameplayTagAddToActor 0 (Variables|Default|GetKnockbackTag)
    (:Added
      (if (Variables|Default|GetToppleOnKnockback)
        (CallFunction|ToppleLabrador)))))

(event Collision|EventActorBeginOverlap
  (params (OtherActor :type Actor%20Object%20Reference)))

(event EventOnDeathServer
  (params (Damage :type Float%20%28single-precision%29) (DamageTags :type Gameplay%20Tag%20Container%20Structure%20%28by%20ref%29) (Momentum :type Vector :default "0, 0, 0") (HitInfo :type Hit%20Result%20Structure%20%28by%20ref%29) (InstigatedBy :type Controller%20Object%20Reference) (DamageCauser :type Actor%20Object%20Reference) (EffectContext :type Gameplay%20Effect%20Context%20Handle%20Structure))
  (|CallParentFunction Damage DamageTags Momentum HitInfo InstigatedBy DamageCauser EffectContext)
  (Collision|SetActorEnableCollision false)
  (AI|SpawnPickupinWorldwithLootTierAsync (Variables|Default|GetLootTableName) (+ (Transformation|GetActorLocation) (Variables|Default|GetLootSpawnOffset))))

(event Custom|ToppleLabrador
  (params (DamageTags :type Gameplay%20Tag%20Container%20Structure))
  ;[1 1 1] Apply Topple
  (if (GameplayTags|HasTag DamageTags "(TagName=\"Gameplay.Damage.Environment.Falling\")")
    (if (not (GameplayTags|HasMatchingGameplayTag "(TagName=\"Riding.IsBeingRidden\")"))
      _)
    (else
      _))
  (if (and (not (Variables|Default|GetIsToppled)) (not (GameplayTags|HasMatchingGameplayTag "(TagName=\"Granted.Riding.Creature.IsBeingRidden\")")))
    (CallFunction|SetIsToppled true)
    (Variables|Default|SetMustBeToppleOnLanding false)))

(event EventOnDeathPlayEffects
  (params (Damage :type Float%20%28single-precision%29) (DamageTags :type Gameplay%20Tag%20Container%20Structure%20%28by%20ref%29) (Momentum :type Vector :default "0, 0, 0") (HitInfo :type Hit%20Result%20Structure%20%28by%20ref%29) (InstigatedBy :type Fort%20Pawn%20Object%20Reference) (DamageCauser :type Actor%20Object%20Reference) (EffectContext :type Gameplay%20Effect%20Context%20Handle%20Structure))
  (|CallParentFunction Damage DamageTags Momentum HitInfo InstigatedBy DamageCauser EffectContext))

(event EventOnLaunched
  (params (LaunchVelocity :type Vector :default "0, 0, 0") (bXYOverride :type Boolean) (bZOverride :type Boolean))
  (|CallParentFunction LaunchVelocity bXYOverride bZOverride)
  (CallFunction|SetIsLaunched true)
  (Variables|Default|SetMustBeToppleOnLanding true))

(event Collision|EventHit
  (params (MyComp :type Primitive%20Component%20Object%20Reference) (Other :type Actor%20Object%20Reference) (OtherComp :type Primitive%20Component%20Object%20Reference) (bSelfMoved :type Boolean) (HitLocation :type Vector :default "0, 0, 0") (HitNormal :type Vector :default "0, 0, 0") (NormalImpulse :type Vector :default "0, 0, 0") (Hit :type Hit%20Result%20Structure%20%28by%20ref%29))
  (|CallParentFunction MyComp Other OtherComp bSelfMoved HitLocation HitNormal NormalImpulse Hit)
  (bind _asfort_athena_vehicle (Utilities|Casting|CastToFortAthenaVehicle Other))
  (if (and (> (Math|Vector|GetMaxElement (Vehicle|GetVehicleVelocity _asfort_athena_vehicle)) 1e+02) (not (GameplayTags|HasMatchingGameplayTag "(TagName=\"Granted.Riding.Creature.IsBeingRidden\")")))
    (CallFunction|ToppleLabrador)))

(event EventOnLanded
  (params (Hit_bBlockingHit :type Boolean) (Hit_bInitialOverlap :type Boolean) (Hit_Time :type Float%20%28single-precision%29) (Hit_Distance :type Float%20%28single-precision%29) (Hit_Location :type Vector :default "0, 0, 0") (Hit_ImpactPoint :type Vector :default "0, 0, 0") (Hit_Normal :type Vector :default "0, 0, 0") (Hit_ImpactNormal :type Vector :default "0, 0, 0") (Hit_PhysMat :type Physical%20Material%20Object%20Reference) (Hit_HitActor :type Actor%20Object%20Reference) (Hit_HitComponent :type Primitive%20Component%20Object%20Reference) (Hit_HitBoneName :type Name :default "None") (Hit_BoneName :type Name :default "None") (Hit_HitItem :type Integer) (Hit_ElementIndex :type Integer) (Hit_FaceIndex :type Integer) (Hit_TraceStart :type Vector :default "0, 0, 0") (Hit_TraceEnd :type Vector :default "0, 0, 0"))
  (|CallParentFunction (Collision|MakeHitResult Hit_bBlockingHit Hit_bInitialOverlap Hit_Time Hit_Distance Hit_Location Hit_ImpactPoint Hit_Normal Hit_ImpactNormal Hit_PhysMat Hit_HitActor Hit_HitComponent Hit_HitBoneName "None" Hit_HitItem Hit_ElementIndex Hit_FaceIndex Hit_TraceStart Hit_TraceEnd))
  (CallFunction|SetIsLaunched false)
  (if (Variables|Default|GetMustBeToppleOnLanding)
    (CallFunction|ToppleLabrador)))

(event WaterBody|EventOnEnteredWaterBody
  (params (WaterBody :type Fort%20Water%20Body%20Actor%20Object%20Reference) (bIsFirstBody :type Boolean))
  (|CallParentFunction WaterBody bIsFirstBody)
  (CallFunction|SetIsLaunched false)
  (CallFunction|SetIsToppled false) ; UnTopple if we enter in water
  (Variables|Default|SetMustBeToppleOnLanding false))

(event Custom|IsLaunchedEvent)

(event Custom|IsToppleEvent)

(event GameplayCue|GameplayCue.Labrador.LeakGolden
  (params (EventType :type EGameplayCueEvent%20Enum :default "OnActive") (Parameters :type Gameplay%20Cue%20Parameters%20Structure%20%28by%20ref%29))
  (bind _returnvalue (Class|FortAnimInputEvent|CreateAnimInputEvent "aie.Labrador.Leak"))
  (Animation|TriggerAnimInputEvent _returnvalue))

(event GameplayCue|GameplayCue.Labrador.Leak
  (params (EventType :type EGameplayCueEvent%20Enum :default "OnActive") (Parameters :type Gameplay%20Cue%20Parameters%20Structure%20%28by%20ref%29))
  (bind _returnvalue (Class|FortAnimInputEvent|CreateAnimInputEvent "aie.Labrador.Leak"))
  (Animation|TriggerAnimInputEvent _returnvalue))

(event EventOnTeleported
  (|CallParentFunction)
  (Variables|Default|SetMustBeToppleOnLanding true))

(event EventOnDamageServer
  (params (Damage :type Float%20%28single-precision%29) (DamageTags :type Gameplay%20Tag%20Container%20Structure%20%28by%20ref%29) (Momentum :type Vector :default "0, 0, 0") (HitInfo :type Hit%20Result%20Structure%20%28by%20ref%29) (InstigatedBy :type Controller%20Object%20Reference) (DamageCauser :type Actor%20Object%20Reference) (EffectContext :type Gameplay%20Effect%20Context%20Handle%20Structure))
  ;[1 1 1] Topple llama on damage.
  (|CallParentFunction Damage DamageTags Momentum HitInfo InstigatedBy DamageCauser EffectContext)
  (CallFunction|PlayHitFX HitInfo)
  (if (and (ScalableFloat|GetValueasBool (Variables|Default|GetToppleOnDamageEnabled) 1.0) (not (Variables|Default|GetToppleBlocked)))
    (if (not (Variables|Default|GetIsToppled))
      (bind _output_get (Variables|Default|SetDamageTaken (+ (Variables|Default|GetDamageTaken) Damage)))
      (if (>= _output_get (Math|Float|SelectFloat (Variables|Default|GetBuffDamageThreshold) (Variables|Default|GetDamageTreshold) (Variables|Default|GetToppleBuffApplied)))
        (Variables|Default|SetDamageTaken 0.0)
        (Variables|Default|SetDamageTreshold (Math|Random|RandomFloatinRange (ScalableFloat|GetValueatLevel (Variables|Default|GetDamageThresholdMin) 1.0) (ScalableFloat|GetValueatLevel (Variables|Default|GetDamageTresholdMax) 1.0)))
        (Utilities|IsValid (WaterBody|GetCurrentWaterBody)
          (:Is%20Valid
            _)
          (:Is%20Not%20Valid
            (Character|LaunchCharacter "0.000000,0.000000,750.000000" false true))) ; Check if already in a water body, if that's the case do not launch as it won't receive enter/exit water body events to stop launch and will never un-stun
        (CallFunction|ToppleLabrador DamageTags)
        (Variables|Default|SetBuffDamageThreshold (Math|Random|RandomFloatinRange (ScalableFloat|GetValueatLevel (Variables|Default|GetBuffDamageThresholdMin) 1.0) (ScalableFloat|GetValueatLevel (Variables|Default|GetBuffDamageThresholdMax) 1.0)))
        (Variables|Default|SetToppleBuffApplied true)
        (Utilities|FlowControl|RetriggerableDelay (ScalableFloat|GetValueatLevel (Variables|Default|GetToppleBuffDuration) 1.0))
        (Variables|Default|SetToppleBuffApplied false)))))

(event Custom|PlayHitFX
  (params (Hit :type Hit%20Result%20Structure))
  (bind (_bblockinghit _binitialoverlap _time _distance _location) (Collision|BreakHitResult Hit))
  (if (Variables|Default|GetShouldPlayHitFX)
    (GameplayCue|ExecuteGameplayCueOnActor(Burst) self (Variables|Default|GetGC_HitFX) (Ability|GameplayCue|MakeGameplayCueParameters :NormalizedMagnitude 0.0 :RawMagnitude 0.0 :Location _location :Normal "0,0,0" :GameplayEffectLevel 1 :AbilityLevel 1 :bReplicateLocationWhenUsingMinimalRepProxy false))))

(fn IsValidAutoFireTarget
  (params (Attacker :type Actor%20Object%20Reference))
  (return true))

(fn OnRep_IsLauched
  (CallFunction|IsLaunchedEvent))

(fn OnRep_IsToppled
  (bind _capsule (Variables|Default|GetCapsule))
  (bind _istoppled (Variables|Default|GetIsToppled))
  (if _istoppled
    (Collision|SetCollisionEnabled _capsule "NoCollision")
    (else
      (Collision|SetCollisionEnabled _capsule "QueryAndPhysics")))
  (CallFunction|IsToppleEvent)
  (bind _returnvalue (Class|FortAnimInputEventBool|CreateAnimInputEventBool "aie.Labrador.IsToppled" _istoppled))
  (Animation|TriggerAnimInputEvent _returnvalue))

(fn SetIsToppled
  (params (IsToppled :type Boolean))
  (bind _self self)
  (bind _returnvalue (AI|GetBlackboard _self))
  (bind _output_get (Variables|Default|SetIsToppled IsToppled))
  (Default|CallToppledDispatcher _output_get)
  (Utilities|IsValid _returnvalue
    (:Is%20Valid
      (AI|Components|Blackboard|SetValueasBool _returnvalue (LocalVariables|GetBBKeyMustBeToppled) _output_get))))

(fn SetIsLaunched
  (params (IsLaunched :type Boolean))
  (bind _self self)
  (bind _returnvalue (AI|GetBlackboard _self))
  (bind _output_get (Variables|Default|SetIsLauched IsLaunched))
  (if (not _output_get)
    (Variables|Default|SetLunchedCurrentAngle 0.0) ; Reset Angle
  )
  (Utilities|IsValid _returnvalue
    (:Is%20Valid
      (AI|Components|Blackboard|SetValueasBool _returnvalue (LocalVariables|GetBBKeyIsLaunched) IsLaunched))))

(fn SetMustJump
  (params (MustJump :type Boolean))
  (bind _self self)
  (bind _returnvalue (AI|GetBlackboard _self))
  (Utilities|IsValid _returnvalue
    (:Is%20Valid
      (AI|Components|Blackboard|SetValueasBool _returnvalue (LocalVariables|GetBBKeyMustJump) MustJump))))

(fn ConstructionScript
  (|CallParentFunction))
