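; Ridable Labrador wildlife pawn that players convert by interacting, with training that turns it into a golden llama piñata.
; Replicated training percentage triggers one-time cosmetics at completion: golden material, level-up FX, sound, and golden hit and death effects.
; A second interaction damages it to invoke destruction, while tags, converter, alert, toppled and piggyback checks gate interaction.
; High-speed landings while ridden kill it unless shockwave-tagged; DropGold spawns gold pickups.
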(event EventBeginPlay
  (|CallParentFunction)
  (Utilities|FlowControl|Sequence
    (:then_0
      (if (>= (Variables|Default|GetTrainingPercentage) 1.0)
        (Networking|CanExecuteCosmeticEvents
          (:True
            (bind _returnvalue (Customization|GetSkeletalMeshforPartType (Variables|/Game/Athena/AI/Pawns/NPC/Parents/NPC_Pawn_Wildlife_Parent.NPC_Pawn_Wildlife_Parent_C|GetFortCharacterParts self) "Body" :!class /Script/FortniteGame.FortCharacterPartsComponent))
            (Rendering|Material|SetMaterial _returnvalue 1 "/Labrador/Meshes/MI_GoldenLama_Pinata_InGame.MI_GoldenLama_Pinata_InGame")))))))

(event EventOnLaunched
  (params (LaunchVelocity :type Vector :default "0, 0, 0") (bXYOverride :type Boolean) (bZOverride :type Boolean))
  (CallFunction|SetIsLaunched true))

(event EventOnLanded
  (params (Hit_bBlockingHit :type Boolean) (Hit_bInitialOverlap :type Boolean) (Hit_Time :type Float%20%28single-precision%29) (Hit_Distance :type Float%20%28single-precision%29) (Hit_Location :type Vector :default "0, 0, 0") (Hit_ImpactPoint :type Vector :default "0, 0, 0") (Hit_Normal :type Vector :default "0, 0, 0") (Hit_ImpactNormal :type Vector :default "0, 0, 0") (Hit_PhysMat :type Physical%20Material%20Object%20Reference) (Hit_HitActor :type Actor%20Object%20Reference) (Hit_HitComponent :type Primitive%20Component%20Object%20Reference) (Hit_HitBoneName :type Name :default "None") (Hit_BoneName :type Name :default "None") (Hit_HitItem :type Integer) (Hit_ElementIndex :type Integer) (Hit_FaceIndex :type Integer) (Hit_TraceStart :type Vector :default "0, 0, 0") (Hit_TraceEnd :type Vector :default "0, 0, 0"))
  (|CallParentFunction (Collision|MakeHitResult Hit_bBlockingHit Hit_bInitialOverlap Hit_Time Hit_Distance Hit_Location Hit_ImpactPoint Hit_Normal Hit_ImpactNormal Hit_PhysMat Hit_HitActor Hit_HitComponent Hit_HitBoneName "None" Hit_HitItem Hit_ElementIndex Hit_FaceIndex Hit_TraceStart Hit_TraceEnd))
  (if (Math|Boolean|ANDBoolean (> (Math|Float|Absolute(Float) (Math|Vector|GetMaxElement (Pawn|Components|CharacterMovement|GetLastUpdateVelocity (Variables|Character|GetCharacterMovement)))) 4e+03) (not (GameplayTags|HasMatchingGameplayTag (Variables|Default|GetShockwaveTag))) (GameplayTags|HasMatchingGameplayTag "(TagName=\"Riding.IsBeingRidden\")"))
    (Game|Damage|ApplyDamage self 1e+07 0 self)))

(event Custom|CosmeticChange
  ; Training
  (Utilities|FlowControl|DoOnce false
    (:Completed
      (bind _returnvalue (Customization|GetSkeletalMeshforPartType (Variables|/Game/Athena/AI/Pawns/NPC/Parents/NPC_Pawn_Wildlife_Parent.NPC_Pawn_Wildlife_Parent_C|GetFortCharacterParts self) "Body" :!class /Script/FortniteGame.FortCharacterPartsComponent))
      (Rendering|Material|SetMaterial _returnvalue 1 "/Labrador/Meshes/MI_GoldenLama_Pinata_InGame.MI_GoldenLama_Pinata_InGame")
      (Effects|SpawnFXSystemAttached "/Labrador/FX/NS_Labrador_Training_LevelUp.NS_Labrador_Training_LevelUp" (Variables|Character|GetMesh))
      (Audio|SpawnSoundAttached :Sound "/Labrador/Sound/Foley/SC_Labrador_TrainingCompleted_Cue.SC_Labrador_TrainingCompleted_Cue" :AttachToComponent (Variables|Character|GetMesh) :AttachPointName "None" :LocationType "KeepRelativeOffset" :bStopWhenAttachedToDestroyed true :VolumeMultiplier 1.0 :PitchMultiplier 1.0 :StartTime 0.0 :bAutoDestroy true)
      (CallFunction|UpdateforGoldenVersion))))

(event Custom|ToggleAnticipationFX
  (params (Enabled :type Boolean))
  (bind _ns_labrador_anticipatetrainingcompleted (Variables|Default|GetNS_Labrador_AnticipateTrainingCompleted))
  ; Training
  (if Enabled
    (Components|Activation|Activate _ns_labrador_anticipatetrainingcompleted false)
    (else
      (Components|Activation|Deactivate _ns_labrador_anticipatetrainingcompleted))))

(event Custom|UpdateForGoldenVersion
  ; Training
  (Variables|Sound|SetDeath%20Normal%20Sound (Variables|Default|GetGoldenDeathSound)) ; death sfx
  (Variables|FortPawn|SetDeathCueTag (Variables|Default|GetGoldenDeathCueTag)) ; death vfx
  (Variables|Default|SetGC_HitFX (Variables|Default|GetGoldenHitFX)))

(event EventOnDeathServer
  (params (Damage :type Float%20%28single-precision%29) (DamageTags :type Gameplay%20Tag%20Container%20Structure%20%28by%20ref%29) (Momentum :type Vector :default "0, 0, 0") (HitInfo :type Hit%20Result%20Structure%20%28by%20ref%29) (InstigatedBy :type Controller%20Object%20Reference) (DamageCauser :type Actor%20Object%20Reference) (EffectContext :type Gameplay%20Effect%20Context%20Handle%20Structure))
  (|CallParentFunction Damage DamageTags Momentum HitInfo InstigatedBy DamageCauser EffectContext))

(event EventOnNPCPawnInteractedWith
  (params (AIPawn :type Actor%20Object%20Reference) (InteractingPlayerPawn :type Fort%20Player%20Pawn%20Object%20Reference) (InteractionBeingAttempted :type EInteractionBeingAttempted%20Enum :default "FirstInteraction"))
  ; Interaction
  (switch Utilities|FlowControl|Switch|SwitchonEInteractionBeingAttempted InteractionBeingAttempted
    (:FirstInteraction
      (|CallParentFunction AIPawn InteractingPlayerPawn "FirstInteraction")
      (if (not (Variables|NPCLeader|GetNPCDynamicBlockConversionOnInteract))
        (bind _asfort_pawn (Utilities|Casting|CastToFortPawn AIPawn))
        (Convert|TriggerConversion InteractingPlayerPawn _asfort_pawn)))
    (:SecondInteraction
      (Default|CallDestroyInvoked)
      (Game|Damage|ApplyDamage self 1e+03))))

(fn BlueprintCanInteract
  (params (InteractingPawn :type Fort%20Pawn%20Object%20Reference) (InteractionBeingAttempted :type EInteractionBeingAttempted%20Enum :default "FirstInteraction"))
  (return (Math|Boolean|ANDBoolean (|CallParentFunction InteractingPawn InteractionBeingAttempted) (not (GameplayTags|HasAnyMatchingGameplayTags InteractingPawn (Variables|Default|GetInteractingPawnPreventingTags))) (select (GameplayTags|HasAnyMatchingGameplayTags (Variables|Default|GetCreative%20Taming%20Allowed%20Tags)) true (not (GameplayTags|HasAnyMatchingGameplayTags (Variables|Default|GetInteraction%20Preventing%20Tags)))) (not (Math|Boolean|ANDBoolean (Utilities|IsValid (Convert|GetConverterPawn (Variables|Default|GetConvertComponent)) :!class /Script/Engine.KismetSystemLibrary) (!= (Convert|GetConverterPawn (Variables|Default|GetConvertComponent)) (Variables|GetInteractingPawn)) (or (Utilities|Enum|Equal(Enum) (Variables|NPCAlertLevel|GetNPC_AlertLevel) "NewEnumerator2") (Utilities|Enum|Equal(Enum) (Variables|NPCAlertLevel|GetNPC_AlertLevel) "NewEnumerator3")))) (not (Variables|Default|GetIsToppled)) (select (ScalableFloat|GetValueasBool (Variables|Default|GetHF_AllowPiggyback)) true (not (GameplayTags|HasMatchingGameplayTag InteractingPawn "(TagName=\"Piggyback.Carrying\")"))))))

(fn DropGold
  (for _ (range (ScalableFloat|GetValueasInteger (Variables|Default|GetHF_GoldStacks)))
    (Inventory|SpawnPickupInWorldWithLevel "/Game/Items/ResourcePickups/Athena_WadsItemData.Athena_WadsItemData" (Loot|GetLootLevel) (ScalableFloat|GetValueasInteger (Variables|Default|GetHF_GoldToSpawn)) (+ (Transformation|GetActorLocation) "0.000000,0.000000,100.000000"))))

(fn OnRep_TrainingPercentage
  (bind _trainingpercentage (Variables|Default|GetTrainingPercentage))
  (Default|CallTrainingIncreased _trainingpercentage)
  (if (>= _trainingpercentage 1.0)
    (CallFunction|CosmeticChange)
    (if false
      (InteractInterface|OverrideSecondInteractionType (Utilities|Enum|LiteralenumTInteractionType "IT_LongPress")))))

(fn ConstructionScript
  (|CallParentFunction))
